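; Example PCG graph that scatters a basic forest over a Mesh Terrain, built entirely from PCG primitive subgraphs.
; Get_MeshTerrain supplies the terrain to Fill_Areas_Randomly, which places random points at the exposed TreeDensity in points per square meter.
; Transform_Project projects those points back onto the same terrain, then Spawn_Assets spawns at the results without setting any assets in this graph.
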(graph MeshTerrain_BasicForest
  (params (TreeDensity :type Double :default 0.005))
  (bind GetMeshTerrain_0 (Subgraph :Graph /PCGPrimitives_MeshPartitionInterop/Primitives/Get/Get_MeshTerrain.Get_MeshTerrain))
  (bind FillAreasRandomly_0 (Subgraph :Graph /PCGPrimitives/Primitives/Fill/Fill_Areas_Randomly.Fill_Areas_Randomly :PointsPerSquaredMeter TreeDensity))
  (bind Subgraph_6 (Subgraph :Graph /PCGPrimitives/Primitives/Transform/Transform_Project.Transform_Project))
  (bind Subgraph_7 (Subgraph :Graph /PCGPrimitives/Primitives/Spawn/Spawn_Assets.Spawn_Assets))
  (connect GetMeshTerrain_0 Out FillAreasRandomly_0 PrimaryInput)
  (connect FillAreasRandomly_0 Out Subgraph_6 PrimaryInput)
  (connect GetMeshTerrain_0 Out Subgraph_6 SecondaryInput)
  (connect Subgraph_6 Out Subgraph_7 PrimaryInput))
