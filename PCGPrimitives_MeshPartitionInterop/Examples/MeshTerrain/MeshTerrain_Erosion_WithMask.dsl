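; Example PCG graph that builds noise hills on a Mesh Terrain and erodes them using a shape-based mask.
; Extracted terrain points get spatial noise, are written at sub-priority 0, then re-extracted for updated normals.
; A spline shape, also debug-displayed, gets a Mask attribute of 1, which is projected onto those points.
; Erosion runs on the masked points and is written at sub-priority 1 in the same "None" priority layer.
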(graph MeshTerrain_Erosion_WithMask
  (bind Subgraph_2 (Subgraph :Graph /PCGPrimitives_MeshPartitionInterop/Primitives/Get/Get_MeshTerrain.Get_MeshTerrain)) ; Get Base terrain
  (bind Subgraph_1 (Subgraph :Graph /PCGPrimitives_MeshPartitionInterop/Primitives/Extract/Extract_MeshTerrainPoints.Extract_MeshTerrainPoints))
  (bind Subgraph_4 (Subgraph :Graph /PCGPrimitives/Primitives/Transform/Transform_Points_SpatialNoise.Transform_Points_SpatialNoise)) ; Create Hills terraforming
  (bind Subgraph_0 (Subgraph :Graph /PCGPrimitives_MeshPartitionInterop/Primitives/Write/Write_MeshTerrain.Write_MeshTerrain)) ; Write Hills into "None" priority layer at sub priority 0.\nReturn updated surface with new positions and rotations.\nWriting and reading back terrain surface allows to get \nupdated rotations allowing us to have updated point normals.
  (bind Subgraph_5 (Subgraph :Graph /PCGPrimitives_MeshPartitionInterop/Primitives/Extract/Extract_MeshTerrainPoints.Extract_MeshTerrainPoints))
  (bind Subgraph_9 (Subgraph :Graph /PCGPrimitives/Primitives/Create/Create_Spline_Shape.Create_Spline_Shape))
  (bind Subgraph_12 (Subgraph :Graph /PCGPrimitives/Primitives/Write/Write_Attribute.Write_Attribute)) ; Create and set Mask attribute to 1
  (bind Subgraph_11 (Subgraph :Graph /PCGPrimitives/Primitives/Transform/Transform_Project.Transform_Project)) ; Create Mask from shape by \nprojecting only the Mask attribute
  (bind Subgraph_3 (Subgraph :Graph /PCGPrimitives_MeshPartitionInterop/Primitives/Transform/Transform_MeshTerrain_Erosion.Transform_MeshTerrain_Erosion)) ; Apply erosion on hills
  (bind Subgraph_6 (Subgraph :Graph /PCGPrimitives_MeshPartitionInterop/Primitives/Write/Write_MeshTerrain.Write_MeshTerrain)) ; Write erosion transform into "None" priority layer at sub priority 1.\nReturn updated surface with new positions and rotations.
  (bind Subgraph_10 (Subgraph :Graph /PCGPrimitives/Primitives/Debug/Debug_PrimitiveData.Debug_PrimitiveData)) ; Display the shape in viewport
  (connect Subgraph_4 Out Subgraph_0 PrimaryInput)
  (connect Subgraph_2 Out Subgraph_1 PrimaryInput)
  (connect Subgraph_11 Out Subgraph_3 PrimaryInput)
  (connect Subgraph_1 Out Subgraph_4 PrimaryInput)
  (connect Subgraph_0 Out Subgraph_5 PrimaryInput)
  (connect Subgraph_3 Out Subgraph_6 PrimaryInput)
  (connect Subgraph_9 Out Subgraph_10 PrimaryInput)
  (connect Subgraph_5 Out Subgraph_11 PrimaryInput)
  (connect Subgraph_12 Out Subgraph_11 SecondaryInput)
  (connect Subgraph_9 Out Subgraph_12 PrimaryInput))
