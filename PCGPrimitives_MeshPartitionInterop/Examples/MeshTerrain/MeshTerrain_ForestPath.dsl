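; Example PCG graph that builds a forest with a path on mesh terrain and reshapes that terrain.
; A jittered spline is outlined and excluded from the traced terrain footprint, then random fills spawn assets in the forest and path areas.
; Terrain points are offset after distance-from-tree and distance-from-path passes, then written back to the mesh terrain.
; Uniformly filling the path with points makes distance measure from its interior, not just its edge.
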(graph MeshTerrain_ForestPath
  ;[0.15 0.15 0.15 0.5] Terrain forest with path
  (bind Subgraph_12 (Subgraph :Graph /PCGPrimitives_MeshPartitionInterop/Primitives/Get/Get_MeshTerrain.Get_MeshTerrain))
  (bind Subgraph_0 (Subgraph :Graph /PCGPrimitives_MeshPartitionInterop/Primitives/Extract/Extract_MeshTerrainPoints.Extract_MeshTerrainPoints))
  (bind Subgraph_1 (Subgraph :Graph /PCGPrimitives/Primitives/Trace/Trace_Bound_Footprint.Trace_Bound_Footprint)) ; This will convert the point from the mesh terrain,\nOnto a spline shape of its footprint.\nThis will enable us to execute an exclusion more easily to later spawn trees. 
  (bind Subgraph_22 (Subgraph :Graph /PCGPrimitives/Primitives/Transform/Transform_Outline.Transform_Outline))
  (bind Subgraph_15 (Subgraph :Graph /PCGPrimitives/Primitives/Transform/Transform_Offset.Transform_Offset))
  (bind Subgraph_9 (Subgraph :Graph /PCGPrimitives/Primitives/Spawn/Spawn_Assets.Spawn_Assets))
  (bind Subgraph_28 (Subgraph :Graph /PCGPrimitives/Primitives/Transform/Transform_Outline.Transform_Outline))
  (bind Subgraph_13 (Subgraph :Graph /PCGPrimitives_MeshPartitionInterop/Primitives/Write/Write_MeshTerrain.Write_MeshTerrain))
  (bind Subgraph_29 (Subgraph :Graph /PCGPrimitives/Primitives/Fill/Fill_Areas_Randomly.Fill_Areas_Randomly))
  (bind Subgraph_26 (Subgraph :Graph /PCGPrimitives/Primitives/Transform/Transform_Offset.Transform_Offset))
  (bind Subgraph_27 (Subgraph :Graph /PCGPrimitives/Primitives/Spawn/Spawn_Assets.Spawn_Assets))
  ;;[0.15 0.15 0.15 0.5] Path
  (bind Subgraph_16 (Subgraph :Graph /PCGPrimitives/Primitives/Create/Create_Spline_line.Create_Spline_line))
  (bind Subgraph_17 (Subgraph :Graph /PCGPrimitives/Primitives/Transform/Transform_Resample.Transform_Resample))
  (bind Subgraph_18 (Subgraph :Graph /PCGPrimitives/Primitives/Transform/Transform_Jitter.Transform_Jitter))
  ;;[0.15 0.15 0.15 0.5] Forest Minus Path
  (bind Subgraph_20 (Subgraph :Graph /PCGPrimitives/Primitives/Compose/Compose_Areas_Exclusion.Compose_Areas_Exclusion))
  (bind Subgraph_7 (Subgraph :Graph /PCGPrimitives/Primitives/Fill/Fill_Areas_Randomly.Fill_Areas_Randomly))
  ;;[0.15 0.07891 0.01782 0.5] ForestSoil & Tree Bumps
  (bind Subgraph_10 (Subgraph :Graph /PCGPrimitives/Primitives/Write/Write_DistanceFrom.Write_DistanceFrom))
  (bind Subgraph_14 (Subgraph :Graph /PCGPrimitives/Primitives/Transform/Transform_Offset.Transform_Offset))
  ;;[0.3177 0.2346 0.1059 0.5] Dirt and Path
  (bind FillAreaUniformly_0 (Subgraph :Graph /PCGPrimitives/Primitives/Fill/Fill_Areas_Uniformly.Fill_Areas_Uniformly)) ; Write_DistanceFrom calculates distance from points,\nnot from a surface or filled area.\nBy filling the path shape uniformly with points,\nwe simulate a surface so that distance is computed\nfrom the interior of the path, not just its edge.
  (bind Subgraph_24 (Subgraph :Graph /PCGPrimitives/Primitives/Write/Write_DistanceFrom.Write_DistanceFrom))
  (bind Subgraph_25 (Subgraph :Graph /PCGPrimitives/Primitives/Transform/Transform_Offset.Transform_Offset))
  (connect Subgraph_28 Out FillAreaUniformly_0 PrimaryInput)
  (connect Subgraph_12 Out Subgraph_0 PrimaryInput)
  (connect Subgraph_0 Out Subgraph_1 PrimaryInput)
  (connect Subgraph_20 Out Subgraph_7 PrimaryInput)
  (connect Subgraph_15 Out Subgraph_9 PrimaryInput)
  (connect Subgraph_0 Out Subgraph_10 PrimaryInput)
  (connect Subgraph_7 Out Subgraph_10 SecondaryInput)
  (connect Subgraph_25 Out Subgraph_13 PrimaryInput)
  (connect Subgraph_10 Out Subgraph_14 PrimaryInput)
  (connect Subgraph_7 Out Subgraph_15 PrimaryInput)
  (connect Subgraph_16 Out Subgraph_17 PrimaryInput)
  (connect Subgraph_17 Out Subgraph_18 PrimaryInput)
  (connect Subgraph_1 Out Subgraph_20 PrimaryInput)
  (connect Subgraph_22 Out Subgraph_20 SecondaryInput)
  (connect Subgraph_18 Out Subgraph_22 PrimaryInput)
  (connect Subgraph_14 Out Subgraph_24 PrimaryInput)
  (connect FillAreaUniformly_0 Out Subgraph_24 SecondaryInput)
  (connect Subgraph_24 Out Subgraph_25 PrimaryInput)
  (connect Subgraph_29 Out Subgraph_26 PrimaryInput)
  (connect Subgraph_26 Out Subgraph_27 PrimaryInput)
  (connect Subgraph_18 Out Subgraph_28 PrimaryInput)
  (connect Subgraph_28 Out Subgraph_29 PrimaryInput))
