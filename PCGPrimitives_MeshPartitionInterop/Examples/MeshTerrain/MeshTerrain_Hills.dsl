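; MeshTerrain_Hills is an example PCG graph that modifies mesh terrain with spatial noise, which its name indicates produces hills.
; It chains four primitive subgraphs linearly: get the mesh terrain, extract its points, transform those points with spatial noise, and write the result back as mesh terrain.
; Terrain access uses MeshPartition interop primitives, while the noise step reuses the general PCGPrimitives transform, all left at default settings.
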(graph MeshTerrain_Hills
  (bind Subgraph_2 (Subgraph :Graph /PCGPrimitives_MeshPartitionInterop/Primitives/Get/Get_MeshTerrain.Get_MeshTerrain))
  (bind Subgraph_1 (Subgraph :Graph /PCGPrimitives_MeshPartitionInterop/Primitives/Extract/Extract_MeshTerrainPoints.Extract_MeshTerrainPoints))
  (bind Subgraph_4 (Subgraph :Graph /PCGPrimitives/Primitives/Transform/Transform_Points_SpatialNoise.Transform_Points_SpatialNoise))
  (bind Subgraph_0 (Subgraph :Graph /PCGPrimitives_MeshPartitionInterop/Primitives/Write/Write_MeshTerrain.Write_MeshTerrain))
  (connect Subgraph_4 Out Subgraph_0 PrimaryInput)
  (connect Subgraph_2 Out Subgraph_1 PrimaryInput)
  (connect Subgraph_1 Out Subgraph_4 PrimaryInput))
