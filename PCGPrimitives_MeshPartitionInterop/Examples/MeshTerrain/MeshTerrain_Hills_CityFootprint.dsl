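; PCG example graph that generates hilly MeshTerrain with a flattened city footprint.
; Terrain points are extracted, displaced by spatial noise and tagged with a Grass attribute.
; A generated city-shape spline carrying a Dirt attribute is projected as a filled 2D polygon, flattening points inside it and transferring Dirt to them.
; The result is written back to MeshTerrain, with Grass and Dirt feeding matching weight channels.
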(graph MeshTerrain_Hills_CityFootprint
  (bind Subgraph_2 (Subgraph :Graph /PCGPrimitives_MeshPartitionInterop/Primitives/Get/Get_MeshTerrain.Get_MeshTerrain))
  (bind Subgraph_7 (Subgraph :Graph /PCGPrimitives_MeshPartitionInterop/Primitives/Extract/Extract_MeshTerrainPoints.Extract_MeshTerrainPoints))
  (bind Subgraph_4 (Subgraph :Graph /PCGPrimitives/Primitives/Transform/Transform_Points_SpatialNoise.Transform_Points_SpatialNoise))
  (bind Subgraph_1 (Subgraph :Graph /PCGPrimitives/Primitives/Create/Create_Spline_Shape.Create_Spline_Shape)) ; City Shape Spline
  (bind Subgraph_6 (Subgraph :Graph /PCGPrimitives/Primitives/Write/Write_Attribute.Write_Attribute)) ; Add Dirt attrib for MeshTerrain
  (bind Subgraph_9 (Subgraph :Graph /PCGPrimitives/Primitives/Write/Write_Attribute.Write_Attribute)) ; Add Grass attrib for \nMeshTerrain weight channel.
  (bind Subgraph_3 (Subgraph :Graph /PCGPrimitives/Primitives/Transform/Transform_Project.Transform_Project)) ; ConvertClosedSplineToPolygon2D: true is set here.\nThis flag treats the closed spline as a filled 2D polygon,\nprojecting terrain points that fall inside its boundary\ndown to a flat plane — effectively flattening the terrain\nwithin the shape, not just snapping to its edge.\nUse this flag whenever you want to flatten a region\ndefined by a closed spline, such as a city footprint,\nplaza, or building pad.\n\nThe Dirt attribute is also transfered from the city shape spline\nand applied to the terrain which will be injected into MeshTerrain \ncorresponding weight channel.
  (bind Subgraph_8 (Subgraph :Graph /PCGPrimitives_MeshPartitionInterop/Primitives/Write/Write_MeshTerrain.Write_MeshTerrain))
  (connect Subgraph_9 Out Subgraph_3 PrimaryInput)
  (connect Subgraph_6 Out Subgraph_3 SecondaryInput)
  (connect Subgraph_7 Out Subgraph_4 PrimaryInput)
  (connect Subgraph_1 Out Subgraph_6 PrimaryInput)
  (connect Subgraph_2 Out Subgraph_7 PrimaryInput)
  (connect Subgraph_3 Out Subgraph_8 PrimaryInput)
  (connect Subgraph_4 Out Subgraph_9 PrimaryInput))
