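; Example PCG graph that builds hills on a mesh terrain and grows a forest over them, composed from PCG primitive subgraphs.
; Terrain points are displaced by spatial noise and written back to form the hills.
; Randomly filled area points are projected onto that terrain, offset and used to spawn assets.
; A second extract-and-write pass sets an attribute and offsets terrain by distance from those points, for grass, forest soil and tree bumps.
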(graph MeshTerrain_Hills_Forest
  ;[0.15 0.15 0.15 0.5] Terrain forest on hills terrain
  (bind Subgraph_2 (Subgraph :Graph /PCGPrimitives_MeshPartitionInterop/Primitives/Get/Get_MeshTerrain.Get_MeshTerrain))
  (bind Subgraph_33 (Subgraph :Graph /PCGPrimitives_MeshPartitionInterop/Primitives/Extract/Extract_MeshTerrainPoints.Extract_MeshTerrainPoints))
  (bind Subgraph_4 (Subgraph :Graph /PCGPrimitives/Primitives/Transform/Transform_Points_SpatialNoise.Transform_Points_SpatialNoise))
  (bind Subgraph_0 (Subgraph :Graph /PCGPrimitives/Primitives/Fill/Fill_Areas_Randomly.Fill_Areas_Randomly))
  (bind Subgraph_28 (Subgraph :Graph /PCGPrimitives_MeshPartitionInterop/Primitives/Write/Write_MeshTerrain.Write_MeshTerrain))
  (bind Subgraph_27 (Subgraph :Graph /PCGPrimitives/Primitives/Transform/Transform_Project.Transform_Project))
  (bind Subgraph_12 (Subgraph :Graph /PCGPrimitives/Primitives/Transform/Transform_Offset.Transform_Offset))
  (bind Subgraph_5 (Subgraph :Graph /PCGPrimitives/Primitives/Spawn/Spawn_Assets.Spawn_Assets))
  (bind Subgraph_32 (Subgraph :Graph /PCGPrimitives_MeshPartitionInterop/Primitives/Extract/Extract_MeshTerrainPoints.Extract_MeshTerrainPoints))
  (bind Subgraph_10 (Subgraph :Graph /PCGPrimitives_MeshPartitionInterop/Primitives/Write/Write_MeshTerrain.Write_MeshTerrain))
  ;;[0.15 0.07891 0.01782 0.5] Grass, ForestSoil & Tree Bumps
  (bind Subgraph_31 (Subgraph :Graph /PCGPrimitives/Primitives/Write/Write_Attribute.Write_Attribute))
  (bind Subgraph_6 (Subgraph :Graph /PCGPrimitives/Primitives/Write/Write_DistanceFrom.Write_DistanceFrom))
  (bind Subgraph_11 (Subgraph :Graph /PCGPrimitives/Primitives/Transform/Transform_Offset.Transform_Offset))
  (connect Subgraph_2 Out Subgraph_0 PrimaryInput)
  (connect Subgraph_33 Out Subgraph_4 PrimaryInput)
  (connect Subgraph_12 Out Subgraph_5 PrimaryInput)
  (connect Subgraph_31 Out Subgraph_6 PrimaryInput)
  (connect Subgraph_27 Out Subgraph_6 SecondaryInput)
  (connect Subgraph_11 Out Subgraph_10 PrimaryInput)
  (connect Subgraph_6 Out Subgraph_11 PrimaryInput)
  (connect Subgraph_27 Out Subgraph_12 PrimaryInput)
  (connect Subgraph_0 Out Subgraph_27 PrimaryInput)
  (connect Subgraph_28 Out Subgraph_27 SecondaryInput)
  (connect Subgraph_4 Out Subgraph_28 PrimaryInput)
  (connect Subgraph_32 Out Subgraph_31 PrimaryInput)
  (connect Subgraph_28 Out Subgraph_32 PrimaryInput)
  (connect Subgraph_2 Out Subgraph_33 PrimaryInput))
