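; Example PCG graph that builds a hilly mesh terrain forest crossed by a road, composed from PCG Primitives subgraphs.
; Spatial noise writes hills, then a resampled, jittered line projects onto them as the road spline for shape grammar.
; Scattered trees are filtered by road distance, projected, offset and spawned as conifers.
; A second terrain pass projects points against the road, applies grass and tree-distance forest soil with bumps, then writes back.
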(graph MeshTerrain_Hills_ForestWithRoad
  (bind GetTerrain (Subgraph :Graph /PCGPrimitives_MeshPartitionInterop/Primitives/Get/Get_MeshTerrain.Get_MeshTerrain))
  (bind ScatterTrees (Subgraph :Graph /PCGPrimitives/Primitives/Fill/Fill_Areas_Randomly.Fill_Areas_Randomly))
  (bind Subgraph_28 (Subgraph :Graph /PCGPrimitives/Primitives/Write/Write_DistanceFrom.Write_DistanceFrom))
  (bind Subgraph_29 (Subgraph :Graph /PCGPrimitives/Primitives/Filter/Filter_AttributeValue.Filter_AttributeValue))
  (bind ProjectTreesToTerrain (Subgraph :Graph /PCGPrimitives/Primitives/Transform/Transform_Project.Transform_Project))
  (bind RaiseTrees (Subgraph :Graph /PCGPrimitives/Primitives/Transform/Transform_Offset.Transform_Offset))
  (bind SpawnConifers (Subgraph :Graph /PCGPrimitives/Primitives/Spawn/Spawn_Assets.Spawn_Assets))
  (bind ExtractTerrainPointsAfterWrite (Subgraph :Graph /PCGPrimitives_MeshPartitionInterop/Primitives/Extract/Extract_MeshTerrainPoints.Extract_MeshTerrainPoints))
  (bind Subgraph_27 (Subgraph :Graph /PCGPrimitives/Primitives/Transform/Transform_Project.Transform_Project))
  (bind ApplyGrassMaterial (Subgraph :Graph /PCGPrimitives/Primitives/Write/Write_Attribute.Write_Attribute))
  (bind WriteForestSoil (Subgraph :Graph /PCGPrimitives/Primitives/Write/Write_DistanceFrom.Write_DistanceFrom))
  (bind ApplyForestSoilBumps (Subgraph :Graph /PCGPrimitives/Primitives/Transform/Transform_Offset.Transform_Offset))
  (bind WriteTerrainMaterials (Subgraph :Graph /PCGPrimitives_MeshPartitionInterop/Primitives/Write/Write_MeshTerrain.Write_MeshTerrain))
  (bind SpawnRoad (Subgraph :Graph /PCGPrimitives/Primitives/Assign/Assign_ShapeGrammarDefinition.Assign_ShapeGrammarDefinition))
  ;[0.15 0.15 0.15 0.5 100] HILLS
  (bind ExtractTerrainPoints (Subgraph :Graph /PCGPrimitives_MeshPartitionInterop/Primitives/Extract/Extract_MeshTerrainPoints.Extract_MeshTerrainPoints))
  (bind AddHills (Subgraph :Graph /PCGPrimitives/Primitives/Transform/Transform_Points_SpatialNoise.Transform_Points_SpatialNoise))
  (bind WriteTerrain (Subgraph :Graph /PCGPrimitives_MeshPartitionInterop/Primitives/Write/Write_MeshTerrain.Write_MeshTerrain))
  ;[0.15 0.15 0.15 0.5 100] ROAD
  (bind RoadLine (Subgraph :Graph /PCGPrimitives/Primitives/Create/Create_Spline_line.Create_Spline_line))
  (bind ResampleRoadSpline (Subgraph :Graph /PCGPrimitives/Primitives/Transform/Transform_Resample.Transform_Resample))
  (bind WigglyRoad (Subgraph :Graph /PCGPrimitives/Primitives/Transform/Transform_Jitter.Transform_Jitter))
  (bind ProjectRoadToTerrain (Subgraph :Graph /PCGPrimitives/Primitives/Transform/Transform_Project.Transform_Project))
  (reroute Hills WriteTerrain)
  (reroute ROAD%20SPLINE ProjectRoadToTerrain)
  (connect ExtractTerrainPoints Out AddHills PrimaryInput)
  (connect WriteForestSoil Out ApplyForestSoilBumps PrimaryInput)
  (connect Subgraph_27 Out ApplyGrassMaterial PrimaryInput)
  (connect GetTerrain Out ExtractTerrainPoints PrimaryInput)
  (connect Hills Out ExtractTerrainPointsAfterWrite PrimaryInput)
  (connect WigglyRoad Out ProjectRoadToTerrain PrimaryInput)
  (connect Hills Out ProjectRoadToTerrain SecondaryInput)
  (connect Subgraph_29 Selected ProjectTreesToTerrain PrimaryInput)
  (connect Hills Out ProjectTreesToTerrain SecondaryInput)
  (connect ProjectTreesToTerrain Out RaiseTrees PrimaryInput)
  (connect RoadLine Out ResampleRoadSpline PrimaryInput)
  (connect GetTerrain Out ScatterTrees PrimaryInput)
  (connect RaiseTrees Out SpawnConifers PrimaryInput)
  (connect ProjectRoadToTerrain Out SpawnRoad PrimaryInput)
  (connect ExtractTerrainPointsAfterWrite Out Subgraph_27 PrimaryInput)
  (connect ROAD%20SPLINE Out Subgraph_27 SecondaryInput)
  (connect ScatterTrees Out Subgraph_28 PrimaryInput)
  (connect ROAD%20SPLINE Out Subgraph_28 SecondaryInput)
  (connect Subgraph_28 Out Subgraph_29 PrimaryInput)
  (connect ResampleRoadSpline Out WigglyRoad PrimaryInput)
  (connect ApplyGrassMaterial Out WriteForestSoil PrimaryInput)
  (connect ProjectTreesToTerrain Out WriteForestSoil SecondaryInput)
  (connect AddHills Out WriteTerrain PrimaryInput)
  (connect ApplyForestSoilBumps Out WriteTerrainMaterials PrimaryInput))
