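; PCG example graph that builds a hilly mesh terrain with a forest around a hexagonal dirt patch.
; Spatial noise displaces extracted terrain points into hills, which are written back as mesh terrain.
; Random scatter points, minus the filled hexagon, are projected onto the hills to spawn assets.
; Re-extracted points get a zero-initialized Dirt attribute for hexagon projection to overwrite, then distance-from-tree data and an offset, before a second terrain write.
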(graph MeshTerrain_Hills_Forest_with_Hexagon_Dirt_Patch
  ;[0.15 0.15 0.15 0.5] Terrain forest on hills terrain
  (bind Subgraph_0 (Subgraph :Graph /PCGPrimitives/Primitives/Fill/Fill_Areas_Randomly.Fill_Areas_Randomly))
  (bind Subgraph_9 (Subgraph :Graph /PCGPrimitives/Primitives/Compose/Compose_Points_Difference.Compose_Points_Difference))
  (bind Subgraph_27 (Subgraph :Graph /PCGPrimitives/Primitives/Transform/Transform_Project.Transform_Project))
  (bind Subgraph_12 (Subgraph :Graph /PCGPrimitives/Primitives/Transform/Transform_Offset.Transform_Offset))
  (bind Subgraph_5 (Subgraph :Graph /PCGPrimitives/Primitives/Spawn/Spawn_Assets.Spawn_Assets))
  (bind Subgraph_3 (Subgraph :Graph /PCGPrimitives/Primitives/Write/Write_Attribute.Write_Attribute)) ; As we will use the hexagon as a way to project a material \nWe need to ensure its attribute `dirt` is set to one. 
  (bind Subgraph_32 (Subgraph :Graph /PCGPrimitives_MeshPartitionInterop/Primitives/Extract/Extract_MeshTerrainPoints.Extract_MeshTerrainPoints))
  (bind Subgraph_31 (Subgraph :Graph /PCGPrimitives/Primitives/Write/Write_Attribute.Write_Attribute))
  (bind Subgraph_14 (Subgraph :Graph /PCGPrimitives/Primitives/Write/Write_Attribute.Write_Attribute)) ; It is necessary to create the disk attribute ahead of time and make sure it is at zero. \nIf not, The attribute projection that follows will have nothing to write to.  
  (bind Subgraph_7 (Subgraph :Graph /PCGPrimitives/Primitives/Write/Write_Attribute_fromProjection.Write_Attribute_fromProjection)) ; The target attribute `Dirt` must exist on\nthe terrain points before this projection runs.\nSee the upstream Write_Attribute node that\ninitializes it to 0 — without it, this\nnode has nothing to write into.
  (bind Subgraph_10 (Subgraph :Graph /PCGPrimitives_MeshPartitionInterop/Primitives/Write/Write_MeshTerrain.Write_MeshTerrain))
  (bind Subgraph_1 (Subgraph :Graph /PCGPrimitives/Primitives/Create/Create_Spline_Shape.Create_Spline_Shape))
  (bind Subgraph_8 (Subgraph :Graph /PCGPrimitives/Primitives/Fill/Fill_Areas_Uniformly.Fill_Areas_Uniformly))
  ;;[0.15 0.15 0.15 0.5] MEsh Terrain Hills
  (bind Subgraph_2 (Subgraph :Graph /PCGPrimitives_MeshPartitionInterop/Primitives/Get/Get_MeshTerrain.Get_MeshTerrain))
  (bind Subgraph_33 (Subgraph :Graph /PCGPrimitives_MeshPartitionInterop/Primitives/Extract/Extract_MeshTerrainPoints.Extract_MeshTerrainPoints))
  (bind Subgraph_4 (Subgraph :Graph /PCGPrimitives/Primitives/Transform/Transform_Points_SpatialNoise.Transform_Points_SpatialNoise))
  (bind Subgraph_28 (Subgraph :Graph /PCGPrimitives_MeshPartitionInterop/Primitives/Write/Write_MeshTerrain.Write_MeshTerrain))
  ;;[0.15 0.07891 0.01782 0.5] Grass, ForestSoil & Tree Bumps
  (bind Subgraph_6 (Subgraph :Graph /PCGPrimitives/Primitives/Write/Write_DistanceFrom.Write_DistanceFrom))
  (bind Subgraph_11 (Subgraph :Graph /PCGPrimitives/Primitives/Transform/Transform_Offset.Transform_Offset))
  (reroute hexagon%20fille%20up%20with%20points Subgraph_8)
  (connect Subgraph_2 Out Subgraph_0 PrimaryInput)
  (connect hexagon%20fille%20up%20with%20points Out Subgraph_3 PrimaryInput)
  (connect Subgraph_33 Out Subgraph_4 PrimaryInput)
  (connect Subgraph_12 Out Subgraph_5 PrimaryInput)
  (connect Subgraph_7 Out Subgraph_6 PrimaryInput)
  (connect Subgraph_27 Out Subgraph_6 SecondaryInput)
  (connect Subgraph_3 Out Subgraph_7 From)
  (connect Subgraph_14 Out Subgraph_7 To)
  (connect Subgraph_1 Out Subgraph_8 PrimaryInput)
  (connect Subgraph_0 Out Subgraph_9 PrimaryInput)
  (connect hexagon%20fille%20up%20with%20points Out Subgraph_9 SecondaryInput)
  (connect Subgraph_11 Out Subgraph_10 PrimaryInput)
  (connect Subgraph_6 Out Subgraph_11 PrimaryInput)
  (connect Subgraph_27 Out Subgraph_12 PrimaryInput)
  (connect Subgraph_31 Out Subgraph_14 PrimaryInput)
  (connect Subgraph_9 Out Subgraph_27 PrimaryInput)
  (connect Subgraph_28 Out Subgraph_27 SecondaryInput)
  (connect Subgraph_4 Out Subgraph_28 PrimaryInput)
  (connect Subgraph_32 Out Subgraph_31 PrimaryInput)
  (connect Subgraph_28 Out Subgraph_32 PrimaryInput)
  (connect Subgraph_2 Out Subgraph_33 PrimaryInput))
