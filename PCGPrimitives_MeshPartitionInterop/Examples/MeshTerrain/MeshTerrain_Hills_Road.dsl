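; PCG example graph that deforms a mesh terrain into hills and lays a road across it.
; Extracted terrain points receive spatial noise and are written back as hills, and a generated spline is projected onto that result.
; Chaining from the first write's output, re-extracted hill points are projected to the spline and written as a second deformation.
; An offset copy of the projected spline independently feeds shape-grammar road mesh spawning.
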(graph MeshTerrain_Hills_Road
  (bind Subgraph_2 (Subgraph :Graph /PCGPrimitives_MeshPartitionInterop/Primitives/Get/Get_MeshTerrain.Get_MeshTerrain))
  (bind Subgraph_9 (Subgraph :Graph /PCGPrimitives_MeshPartitionInterop/Primitives/Extract/Extract_MeshTerrainPoints.Extract_MeshTerrainPoints))
  (bind Subgraph_4 (Subgraph :Graph /PCGPrimitives/Primitives/Transform/Transform_Points_SpatialNoise.Transform_Points_SpatialNoise))
  (bind Subgraph_0 (Subgraph :Graph /PCGPrimitives_MeshPartitionInterop/Primitives/Write/Write_MeshTerrain.Write_MeshTerrain))
  (bind Subgraph_1 (Subgraph :Graph /PCGPrimitives/Primitives/Create/Create_Spline_Shape.Create_Spline_Shape))
  (bind Subgraph_3 (Subgraph :Graph /PCGPrimitives/Primitives/Transform/Transform_Project.Transform_Project))
  (bind Subgraph_10 (Subgraph :Graph /PCGPrimitives_MeshPartitionInterop/Primitives/Extract/Extract_MeshTerrainPoints.Extract_MeshTerrainPoints)) ; This Extract reads from the output of the\nfirst Write_MeshTerrain (the hills deformation),\nnot from the original flat terrain.\nThis is intentional: we want the road to\nconform to the hills shape, so we need the\nterrain points AFTER the hills have been applied.\nWrite_MeshTerrain returns the updated surface,\nenabling this chaining pattern.
  (bind Subgraph_7 (Subgraph :Graph /PCGPrimitives/Primitives/Transform/Transform_Project.Transform_Project))
  (bind Subgraph_8 (Subgraph :Graph /PCGPrimitives_MeshPartitionInterop/Primitives/Write/Write_MeshTerrain.Write_MeshTerrain))
  (bind Subgraph_12 (Subgraph :Graph /PCGPrimitives/Primitives/Transform/Transform_Offset.Transform_Offset))
  (bind Subgraph_6 (Subgraph :Graph /PCGPrimitives/Primitives/Assign/Assign_ShapeGrammarDefinition.Assign_ShapeGrammarDefinition)) ; Road mesh spawning. Assign_GrammarAsset spawns\ngeometry directly into the scene and does not\nneed to go through Write_MeshTerrain.\nIt is independent of the terrain deformation pipeline.
  (connect Subgraph_4 Out Subgraph_0 PrimaryInput)
  (connect Subgraph_1 Out Subgraph_3 PrimaryInput)
  (connect Subgraph_0 Out Subgraph_3 SecondaryInput)
  (connect Subgraph_9 Out Subgraph_4 PrimaryInput)
  (connect Subgraph_12 Out Subgraph_6 PrimaryInput)
  (connect Subgraph_10 Out Subgraph_7 PrimaryInput)
  (connect Subgraph_3 Out Subgraph_7 SecondaryInput)
  (connect Subgraph_7 Out Subgraph_8 PrimaryInput)
  (connect Subgraph_2 Out Subgraph_9 PrimaryInput)
  (connect Subgraph_0 Out Subgraph_10 PrimaryInput)
  (connect Subgraph_3 Out Subgraph_12 PrimaryInput))
