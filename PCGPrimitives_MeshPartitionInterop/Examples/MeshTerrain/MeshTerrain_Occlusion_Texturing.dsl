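; Example PCG graph that shapes a mesh terrain into eroded hills and textures it with grass and dirt based on occlusion.
; Spatial noise and erosion are written as successive sub-priority layers, each returning the updated surface for point re-extraction with fresh normals.
; Ambient occlusion is computed, then grass is filled and an Occlusion-driven lerp produces Dirt before both weight channels are written.
; An occlusion compare branch also targets Dirt, but its output is unconnected.
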(graph MeshTerrain_Occlusion_Texturing
  (bind Subgraph_2 (Subgraph :Graph /PCGPrimitives_MeshPartitionInterop/Primitives/Get/Get_MeshTerrain.Get_MeshTerrain)) ; Get Base terrain
  (bind Subgraph_1 (Subgraph :Graph /PCGPrimitives_MeshPartitionInterop/Primitives/Extract/Extract_MeshTerrainPoints.Extract_MeshTerrainPoints))
  (bind Subgraph_4 (Subgraph :Graph /PCGPrimitives/Primitives/Transform/Transform_Points_SpatialNoise.Transform_Points_SpatialNoise)) ; Create Hills terraforming
  (bind Subgraph_0 (Subgraph :Graph /PCGPrimitives_MeshPartitionInterop/Primitives/Write/Write_MeshTerrain.Write_MeshTerrain)) ; Write Hills into "None" priority layer at sub priority 0.\nReturn updated surface with new positions and rotations.\nWriting and reading back terrain surface allows to get \nupdated rotations allowing us to have updated point normals.
  (bind Subgraph_5 (Subgraph :Graph /PCGPrimitives_MeshPartitionInterop/Primitives/Extract/Extract_MeshTerrainPoints.Extract_MeshTerrainPoints))
  (bind Subgraph_3 (Subgraph :Graph /PCGPrimitives_MeshPartitionInterop/Primitives/Transform/Transform_MeshTerrain_Erosion.Transform_MeshTerrain_Erosion)) ; Apply erosion on hills
  (bind Subgraph_6 (Subgraph :Graph /PCGPrimitives_MeshPartitionInterop/Primitives/Write/Write_MeshTerrain.Write_MeshTerrain)) ; Write erosion transform into "None" priority layer at sub priority 1.\nReturn updated surface with new positions and rotations.
  (bind Subgraph_8 (Subgraph :Graph /PCGPrimitives_MeshPartitionInterop/Primitives/Extract/Extract_MeshTerrainPoints.Extract_MeshTerrainPoints))
  (bind Subgraph_7 (Subgraph :Graph /PCGPrimitives_MeshPartitionInterop/Primitives/Write/Write_MeshTerrain_AO.Write_MeshTerrain_AO))
  (bind Subgraph_9 (Subgraph :Graph /PCGPrimitives/Primitives/Write/Write_Attribute.Write_Attribute)) ; Fill with grass
  (bind AttributeCompareOp_0 (Attribute%20Compare%20Op :InputSource1 Occlusion :Operation Less :OutputTarget Dirt))
  (bind AttributeMathsOp_0 (Attribute%20Maths%20Op :InputSource3 Occlusion :Operation Lerp :OutputTarget Dirt))
  (bind Subgraph_11 (Subgraph :Graph /PCGPrimitives_MeshPartitionInterop/Primitives/Write/Write_MeshTerrain.Write_MeshTerrain)) ; Write Grass and Dirt weight channels
  (connect Subgraph_9 Out AttributeCompareOp_0 InA)
  (connect Subgraph_9 Out AttributeMathsOp_0 InA)
  (connect Subgraph_9 Out AttributeMathsOp_0 Ratio)
  (connect Subgraph_4 Out Subgraph_0 PrimaryInput)
  (connect Subgraph_2 Out Subgraph_1 PrimaryInput)
  (connect Subgraph_5 Out Subgraph_3 PrimaryInput)
  (connect Subgraph_1 Out Subgraph_4 PrimaryInput)
  (connect Subgraph_0 Out Subgraph_5 PrimaryInput)
  (connect Subgraph_3 Out Subgraph_6 PrimaryInput)
  (connect Subgraph_8 Out Subgraph_7 PrimaryInput)
  (connect Subgraph_6 Out Subgraph_8 PrimaryInput)
  (connect Subgraph_7 Out Subgraph_9 PrimaryInput)
  (connect AttributeMathsOp_0 Out Subgraph_11 PrimaryInput))
