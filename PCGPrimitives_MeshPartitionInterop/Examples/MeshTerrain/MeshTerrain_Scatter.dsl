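; PCG example graph that scatters assets, two sample tree meshes by default, across a mesh terrain within an actor's footprint.
; A single point from the actor feeds a bound-footprint trace, whose area is randomly filled at a configurable per-square-meter density.
; The points are projected onto the fetched mesh terrain, then spawned with position, rotation and scale offsets.
; Random transform range parameters are declared but not wired into any node.
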(graph MeshTerrain_Scatter
  (params (PointsPerSqMeter :type Double :default 0.005) (Assets :type (Array (Struct /Script/CoreUObject.SoftObjectPath)) :default [(struct :AssetPath (struct :PackageName /PCGBiomeSample/Meshes/PCG_Tree_01 :AssetName PCG_Tree_01)) (struct :AssetPath (struct :PackageName /PCGBiomeSample/Meshes/PCG_Tree_02 :AssetName PCG_Tree_02))]) (PositionOffset :type Vector) (RotationOffset :type Rotator) (ScaleOffset :type Vector :default (1.0 1.0 1.0)) (RandomPositionRange :type Vector) (RandomRotationRange :type Vector) (RandomScaleRange :type Vector))
  (bind GetMeshTerrain_0 (Subgraph :Graph /PCGPrimitives_MeshPartitionInterop/Primitives/Get/Get_MeshTerrain.Get_MeshTerrain))
  (bind GetActorData_0 (Get%20Actor%20Data :Mode GetSinglePoint))
  (bind Subgraph_1 (Subgraph :Graph /PCGPrimitives/Primitives/Trace/Trace_Bound_Footprint.Trace_Bound_Footprint))
  (bind FillAreasRandomly_0 (Subgraph :Graph /PCGPrimitives/Primitives/Fill/Fill_Areas_Randomly.Fill_Areas_Randomly :PointsPerSquaredMeter PointsPerSqMeter))
  (bind Subgraph_2 (Subgraph :Graph /PCGPrimitives/Primitives/Transform/Transform_Project.Transform_Project))
  (bind Subgraph_7 (Subgraph :Graph /PCGPrimitives/Primitives/Spawn/Spawn_Assets.Spawn_Assets :Assets Assets :PositionOffset PositionOffset :RotationOffset RotationOffset :ScaleOffset ScaleOffset))
  (connect Subgraph_1 Out FillAreasRandomly_0 PrimaryInput)
  (connect GetActorData_0 Out Subgraph_1 PrimaryInput)
  (connect FillAreasRandomly_0 Out Subgraph_2 PrimaryInput)
  (connect GetMeshTerrain_0 Out Subgraph_2 SecondaryInput)
  (connect Subgraph_2 Out Subgraph_7 PrimaryInput))
